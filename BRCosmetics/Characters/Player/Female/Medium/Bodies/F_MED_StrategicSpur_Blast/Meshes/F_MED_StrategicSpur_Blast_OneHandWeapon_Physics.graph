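# Physics asset for the F_MED_StrategicSpur_Blast female character's one-hand weapon, setting up a dangling left-side megaphone with body collision.
# The megaphone base locks to the pelvis and carries a two-link dynamic chain with limited swing and zero-degree twist limits.
# Capsules on the pelvis, lower spine, left forearm, hand and thigh twist bone approximate the nearby body, while thigh_l has no shapes and serves as a constraint link.
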PhysicsAsset /BRCosmetics/Characters/Player/Female/Medium/Bodies/F_MED_StrategicSpur_Blast/Meshes/F_MED_StrategicSpur_Blast_OneHandWeapon_Physics
bodies {
 base_megaphone_01_l [{shapeType=Capsule, length=0.505, radius=0.505, shapeName=base_megaphone_01_l_capsule, rotation=<rot 90 0 0>}]
 dyn_megaphone_01_l [{
   shapeType=Capsule
   length=3.2885
   radius=2.4377
   shapeName=dyn_megaphone_01_l_capsule
   center=<3.3975 0 1.7884>
   rotation=<rot 72.5656 -0.0003 -0.0004>
  }]
 dyn_megaphone_02_l [
  {
   shapeType=Capsule
   length=10.7919
   radius=7.2457
   shapeName=dyn_megaphone_02_l_capsule
   center=<12.0339 0.2065 -1.6374>
   rotation=<rot 60.6022 0.1557 -4.729>
  }
  {
   shapeType=Capsule
   length=8.0066
   radius=3.3139
   shapeName=dyn_megaphone_02_l_capsule
   center=<6.1442 -9.34 -2.5761>
   rotation=<rot -7.7978 -0.0077 -96.9741>
  }
 ]
 hand_l [{shapeType=Capsule, length=6.1167, radius=5.8137, shapeName=hand_l_capsule, center=<6.5688 16.2179 0>, rotation=<rot 90 0 0>}]
 lowerarm_l [{
   shapeType=Capsule
   length=29.5078
   radius=4.013
   shapeName=lowerarm_l_capsule
   center=<16.6569 0.0006 5.8382>
   rotation=<rot 87.1821 -0.8598 -0.917>
  }]
 pelvis [{shapeType=Capsule, length=13.2821, radius=8.9678, shapeName=pelvis_capsule, center=<-2.4029 -1.9013 0.8261>}]
 spine_01 [{shapeType=Capsule, length=13.6192, radius=7.5705, shapeName=spine_01_capsule, center=<1.4749 -0.4624 0>, rotation=<rot 0 -90 0>}]
 thigh_l []
 thigh_twist_01_l [{
   shapeType=Capsule
   length=29.3229
   radius=8.1771
   shapeName=thigh_l_tapered_capsule
   center=<6.4713 -0.6142 1.8854>
   rotation=<rot 87.6465 179.9992 179.9995>
  }]
}
constraints [
 {bone1Name=dyn_megaphone_01_l, bone2Name=base_megaphone_01_l, swing1LimitDegrees=50, swing2LimitDegrees=50, twistLimitDegrees=0}
 {bone1Name=dyn_megaphone_02_l, bone2Name=dyn_megaphone_01_l, swing1LimitDegrees=30, swing2LimitDegrees=30, twistLimitDegrees=0}
 {
  bone1Name=thigh_l
  bone2Name=pelvis
  swing1LimitDegrees=45
  swing1Motion=Locked
  swing2LimitDegrees=45
  swing2Motion=Locked
  twistLimitDegrees=45
  twistMotion=Locked
 }
 {
  bone1Name=base_megaphone_01_l
  bone2Name=pelvis
  swing1LimitDegrees=45
  swing1Motion=Locked
  swing2LimitDegrees=45
  swing2Motion=Locked
  twistLimitDegrees=45
  twistMotion=Locked
 }
 {
  bone1Name=hand_l
  bone2Name=lowerarm_l
  swing1LimitDegrees=45
  swing1Motion=Locked
  swing2LimitDegrees=45
  swing2Motion=Locked
  twistLimitDegrees=45
  twistMotion=Locked
 }
 {
  bone1Name=lowerarm_l
  bone2Name=spine_01
  swing1LimitDegrees=45
  swing1Motion=Locked
  swing2LimitDegrees=45
  swing2Motion=Locked
  twistLimitDegrees=45
  twistMotion=Locked
 }
 {
  bone1Name=spine_01
  bone2Name=pelvis
  swing1LimitDegrees=45
  swing1Motion=Free
  swing2LimitDegrees=45
  swing2Motion=Free
  twistLimitDegrees=45
  twistMotion=Free
 }
 {
  bone1Name=thigh_twist_01_l
  bone2Name=thigh_l
  swing1LimitDegrees=45
  swing1Motion=Free
  swing2LimitDegrees=45
  swing2Motion=Free
  twistLimitDegrees=45
  twistMotion=Free
 }
]
