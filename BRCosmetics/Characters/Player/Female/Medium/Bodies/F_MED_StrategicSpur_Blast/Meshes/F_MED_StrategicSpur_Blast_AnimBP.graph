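# Post-copy AnimBP for the StrategicSpur Blast female body cosmetic, adding Control Rig correction and physics to the pose copied from the attached parent mesh.
# Control Rig takes crouch, DBNO and large-gun values from the EventGraph, then feeds two RigidBody branches with separate physics assets, pelvis-space simulation and gravity override.
# Additive offsets adjust megaphone bones in both branches.
# SecondDyn selects the one-hand branch for one-hand or dual-wield melee outside swimming and skydiving.
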AnimBlueprint /BRCosmetics/Characters/Player/Female/Medium/Bodies/F_MED_StrategicSpur_Blast/Meshes/F_MED_StrategicSpur_Blast_AnimBP
target_skeleton=&/F_MED_StrategicSpur_Blast_Skeleton
graphs [
 AnimGraph {nodes [
   Output Pose(Result=Blend Poses by bool(
     BlendPose_0=Use cached pose 'CRCache_OneHandWeapon'()
     BlendPose_1=Use cached pose 'CRCache_TwoHandWeapon'()
     bActiveValue=Get SecondDyn() -> (SecondDyn:bool)
    ) {blendType=Linear}) {layerGroup=None}
   n1 = Control Rig!(
    Source=Copy Pose From Mesh() {bUseAttachedParent=true}
    Alpha
    crouchCTRL:double=Get CrouchAlpha() -> (CrouchAlpha:double)
    LargeGunEqquiped:double=Get large_gun_out() -> (large_gun_out:double)
    DBNOSet:double=Get dbnoAlpha() -> (dbnoAlpha:double)
   ) -> (Pose) {
    alphaBoolBlend {blendOption=Linear}
    controlRigAssetReference {blueprintRigClass=&/F_MED_StrategicSpur_Blast_CtrlRig.F_MED_StrategicSpur_Blast_CtrlRig_C}
   }
   Save cached pose 'CRCache_Input'(Pose=@n1) {cacheName=CRCache_Input}
   n2 = RigidBody!(
    ComponentPose=Local To Component(LocalPose=Use cached pose 'CRCache_Input'())
    Alpha
    OverrideWorldGravity:Vector=Get GravityOverride()
    ExternalForce
   ) -> (Pose) {
    bOverrideWorldGravity=true
    overridePhysicsAsset=&/F_MED_StrategicSpur_Blast_Physics
    simulationSpace=BaseBoneSpace
    alphaBoolBlend {blendOption=Linear}
    baseBoneRef {boneName=pelvis}
    componentAppliedLinearAccClamp=<3000 3000 3000>
    componentLinearAccScale=<0.3 0.3 0.3>
    componentLinearVelScale=<0.3 0.3 0.3>
    simSpaceSettings {
     maxAngularAcceleration=1000
     maxAngularVelocity=3
     maxLinearAcceleration=700
     maxLinearVelocity=500
     velocityScaleZ=0
     worldAlpha=0.2
     externalLinearDragV=<0.2 0.2>
    }
   }
   n3 = "Transform (Modify) Bone - Bone: base_megaASW_01_l"(ComponentPose=@n2, Alpha=1, Translation="-0.15,0,-0.3", Rotation="-4,0,0", Scale="1,1,1") {
    rotationMode=BMM_Additive
    rotationSpace=BCS_BoneSpace
    translationMode=BMM_Additive
    translationSpace=BCS_BoneSpace
    alphaBoolBlend {blendOption=Linear}
    alphaScaleBiasClamp {bInterpResult=true}
    boneToModify {boneName=base_megaASW_01_l}
    rotation {pitch=-4}
    translation {x=-0.15, z=-0.3}
   }
   n4 = "Constraint - base_megaASW_01_l (PP)"!(ComponentPose=@n3, bAlphaBoolEnabled:bool=True, ConstraintWeights_0=0.5, ConstraintWeights_1:float=0.5) -> (Pose) {
    alphaInputType=Bool
    alphaBoolBlend {blendOption=Linear}
    boneToModify {boneName=base_megaASW_01_l}
    constraintSetup [{targetBone {boneName=pelvis}}, {targetBone {boneName=thigh_l}}]
    constraintWeights [1,1]
   }
   n5 = "Transform (Modify) Bone - Bone: base_megaASW_01_l"!(ComponentPose=@n4, bAlphaBoolEnabled:bool=True, Translation="-0.5,1.6,-0.6", Rotation="1.3,16.8,7.2", Scale="1,1,1") -> (Pose) {
    alphaInputType=Bool
    rotationMode=BMM_Additive
    rotationSpace=BCS_BoneSpace
    translationMode=BMM_Additive
    translationSpace=BCS_BoneSpace
    alphaBoolBlend {blendOption=Linear}
    alphaScaleBiasClamp {bInterpResult=true}
    boneToModify {boneName=base_megaASW_01_l}
    rotation=<rot 1.3 16.8 7.2>
    translation=<-0.5 1.6 -0.6>
   }
   n6 = "Transform (Modify) Bone - Bone: dyn_megaphone_02_l"!(ComponentPose=@n5, bAlphaBoolEnabled:bool=True, Translation="-0.148,-0.140,1.064", Rotation="4.080,-43.851,-10.860", Scale="1,1,1") -> (Pose) {
    alphaInputType=Bool
    rotationMode=BMM_Additive
    rotationSpace=BCS_BoneSpace
    translationMode=BMM_Additive
    translationSpace=BCS_BoneSpace
    alphaBoolBlend {blendOption=Linear}
    alphaScaleBiasClamp {bInterpResult=true}
    boneToModify {boneName=dyn_megaphone_02_l}
    rotation=<rot 4.08 -43.851 -10.86>
    translation=<-0.148 -0.14 1.0643>
   }
   Save cached pose 'CRCache_TwoHandWeapon'(Pose=Component To Local(ComponentPose=@n6)) {cacheName=CRCache_TwoHandWeapon}
   n7 = "Transform (Modify) Bone - Bone: base_megaphone_01_l"(
    ComponentPose=Local To Component(LocalPose=Use cached pose 'CRCache_Input'())
    Alpha=1
    Translation="0,0,0"
    Rotation="-2.631,-40.519,-6.240"
    Scale="1,1,1"
   ) {
    rotationMode=BMM_Additive
    rotationSpace=BCS_BoneSpace
    scaleMode=BMM_Additive
    scaleSpace=BCS_BoneSpace
    translationMode=BMM_Additive
    translationSpace=BCS_BoneSpace
    alphaBoolBlend {blendOption=Linear}
    alphaScaleBiasClamp {bInterpResult=true}
    boneToModify {boneName=base_megaphone_01_l}
    rotation=<rot -2.631 -40.519 -6.24>
   }
   n8 = RigidBody(ComponentPose=@n7, OverrideWorldGravity=Get GravityOverride()) {
    bOverrideWorldGravity=true
    overridePhysicsAsset=&/F_MED_StrategicSpur_Blast_OneHandWeapon_Physics
    simulationSpace=BaseBoneSpace
    alphaBoolBlend {blendOption=Linear}
    baseBoneRef {boneName=pelvis}
    componentAppliedLinearAccClamp=<3000 3000 3000>
    componentLinearAccScale=<0.3 0.3 0.3>
    componentLinearVelScale=<0.3 0.3 0.3>
    simSpaceSettings {
     maxAngularAcceleration=1000
     maxAngularVelocity=3
     maxLinearAcceleration=700
     maxLinearVelocity=500
     velocityScaleZ=0
     worldAlpha=0.2
     externalLinearDragV=<0.2 0.2>
    }
   }
   n9 = "Transform (Modify) Bone - Bone: base_megaphone_01_l"(ComponentPose=@n8, Alpha=1, Translation="0,0,0", Rotation="-8.323,-21.953,27.298", Scale="1,1,1") {
    rotationMode=BMM_Additive
    rotationSpace=BCS_BoneSpace
    scaleMode=BMM_Additive
    scaleSpace=BCS_BoneSpace
    translationMode=BMM_Additive
    translationSpace=BCS_BoneSpace
    alphaBoolBlend {blendOption=Linear}
    alphaScaleBiasClamp {bInterpResult=true}
    boneToModify {boneName=base_megaphone_01_l}
    rotation=<rot -8.323 -21.953 27.298>
   }
   n10 = "Transform (Modify) Bone - Bone: base_megaASW_01_l"!(ComponentPose=@n9, Alpha:float=1, Translation="-0.15,0,-0.3", Rotation="-4,0,0", Scale="1,1,1") -> (Pose) {
    rotationMode=BMM_Additive
    rotationSpace=BCS_BoneSpace
    translationMode=BMM_Additive
    translationSpace=BCS_BoneSpace
    alphaBoolBlend {blendOption=Linear}
    alphaScaleBiasClamp {bInterpResult=true}
    boneToModify {boneName=base_megaASW_01_l}
    rotation {pitch=-4}
    translation {x=-0.15, z=-0.3}
   }
   n11 = "Transform (Modify) Bone - Bone: dyn_megaphone_02_l"(ComponentPose=@n10, bAlphaBoolEnabled=True, Translation="-0.148,-0.140,1.064", Rotation="4.080,-43.851,-10.860", Scale="1,1,1") {
    alphaInputType=Bool
    rotationMode=BMM_Additive
    rotationSpace=BCS_BoneSpace
    translationMode=BMM_Additive
    translationSpace=BCS_BoneSpace
    alphaBoolBlend {blendOption=Linear}
    alphaScaleBiasClamp {bInterpResult=true}
    boneToModify {boneName=dyn_megaphone_02_l}
    rotation=<rot 4.08 -43.851 -10.86>
    translation=<-0.148 -0.14 1.0643>
   }
   Save cached pose 'CRCache_OneHandWeapon'(Pose=Component To Local(ComponentPose=@n11)) {cacheName=CRCache_OneHandWeapon}
  ]}
 EventGraph {nodes [
   Event BlueprintUpdateAnimation()
   > n1 = Sequence!(execute) -> (then_0, then_1, then_2:Exec)
   n2 = Sequence!(execute=@n1.then_1) -> (then_0, then_1)
   n3 = OR Boolean!(A=Get bIsCrouching(), B=Get bIsSprinting(), C:bool=Get bIsSkydiveDiveMode()) -> (ReturnValue)
   Set megaphoneASW(execute=@n2.then_1, megaphoneASW:double="To Float (Boolean)"(InBool=NOT Boolean(A=@n3))) -> (Output_Get:double)
   > Set notDyn(notDyn:bool=NOT Boolean(A=Get bRigidBodyEnabled())) -> (Output_Get:bool)
   Set CrouchAlpha(execute=@n2.then_0, CrouchAlpha:double="To Float (Boolean)"(InBool=Get bIsCrouching())) -> (Output_Get:double)
   > Set dbnoAlpha(dbnoAlpha:double="To Float (Boolean)"(InBool=Get bIsDBNO())) -> (Output_Get:double)
   n4 = Sequence!(execute=@n1.then_2) -> (then_0, then_1, then_2:Exec)
   > n5 = Cast To FortPlayerAnimInstance(Object=GetAnimInstance(self=GetOwnerSkeletalMeshComponent()))
   > Set rifle_out(rifle_out:bool=false) -> (Output_Get:bool)
   > Set MeeleOneHand(MeeleOneHand:bool=false) -> (Output_Get:bool)
   > n6 = Switch on EFortWeaponCoreAnimation!(execute, Selection:EFortWeaponCoreAnimation Enum=Get WeaponCoreAnim(self=@n5)) -> (
    Melee:Exec
    Pistol:Exec
    Shotgun:Exec
    PaperBlueprint:Exec
    Rifle:Exec
    MeleeOneHand:Exec
    MachinePistol:Exec
    RocketLauncher:Exec
    GrenadeLauncher:Exec
    GoingCommando:Exec
    AssaultRifle:Exec
    TacticalShotgun:Exec
    SniperRifle:Exec
    TrapPlacement:Exec
    ShoulderLauncher:Exec
    AbilityDecoTool:Exec
    Crossbow:Exec
    C4:Exec
    RemoteControl:Exec
    DualWield:Exec
    AR_BullPup:Exec
    AR_ForwardGrip:Exec
    MedPackPaddles:Exec
    SMG_P90:Exec
    AR_DrumGun:Exec
    Consumable_Small:Exec
    Consumable_Large:Exec
    Balloon:Exec
    MountedTurret:Exec
    CreativeTool:Exec
    ExplosiveBow:Exec
    AshtonIndigo:Exec
    AshtonChicago:Exec
    MeleeDualWield:Exec
    Unarmed:Exec
   )
   Branch(execute=[@n6.AssaultRifle, @n6.Shotgun])
   > Set rifle_out(rifle_out=true)
   Set MeeleOneHand(execute=[@n6.MeleeDualWield, @n6.MeleeOneHand], MeeleOneHand=true)
   Set large_gun_out(execute=@n4.then_1, large_gun_out:double="To Float (Boolean)"(InBool=Get rifle_out() -> (rifle_out:bool))) -> (Output_Get:double)
   n7 = Get bIsSkyDiving()
   n8 = Get bIsSurfaceSwimming()
   Set SecondDyn(
    execute=@n4.then_2
    SecondDyn:bool=AND Boolean!(A=Get MeeleOneHand() -> (MeeleOneHand:bool), B=NOT Boolean(A=@n8), C:bool=NOT Boolean(A=@n7)) -> (ReturnValue)
   ) -> (Output_Get:bool)
   n9 = Gravity Override State Function DEPRECATED(
    execute=@n1.then_0
    CharacterPartAnimInstance=Self-Reference()
    GameTimeInSeconds=GetGameTimeInSeconds()
    GravityOverrideParams=Get Gravity Override Params()
   )
   > Set GravityOverride(GravityOverride=@n9)
   Event OnEndCrouching()
   > Set notCrouching(notCrouching:double=1) -> (Output_Get:double)
   Event OnBeginCrouching()
   > Set notCrouching(notCrouching=0)
   Reroute Node(InputPin:Wildcard) -> (OutputPin:Wildcard)
  ]}
]
variables [
 CrouchAlpha {type=number}
 dbnoAlpha {type=number}
 Gravity Override Params {type=GravityOverrideParamsStruct}
 GravityOverride {type=Vector}
 large_gun_out {type=number}
 MeeleOneHand {type=boolean}
 megaphoneASW {type=number}
 notCrouching {type=number}
 notDyn {type=boolean}
 rifle_out {type=boolean}
 SecondDyn {type=boolean}
]
