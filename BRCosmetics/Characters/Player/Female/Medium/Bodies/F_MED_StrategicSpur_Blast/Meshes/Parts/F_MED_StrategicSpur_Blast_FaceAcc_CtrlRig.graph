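# Corrective Control Rig for the F_MED_StrategicSpur_Blast face accessory, repositioning base hair bones as the head rotates.
# Two spherical pose readers on the head bone, differing in rotation offset, each drive a Modify Transforms pose on the hair.
# The first pose is weighted by notCrouch and the second by crouchStatus, giving separate standing and crouched corrections.
# Both readers have debug drawing enabled.
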ControlRigBlueprint /BRCosmetics/Characters/Player/Female/Medium/Bodies/F_MED_StrategicSpur_Blast/Meshes/Parts/F_MED_StrategicSpur_Blast_FaceAcc_CtrlRig
graphs [
 RigVMFunctionLibrary {links [], nodes []}
 RigVMModel {nodes [
   RigUnit_BeginExecution = &RigUnit_BeginExecution()
   n1 = &RigUnit_SphericalPoseReader(
    ActiveRegionScaleFactors=(PositiveWidth=1.5,NegativeWidth=0.2,PositiveHeight=1,NegativeHeight=1)
    ActiveRegionSize=0.1
    Debug=(bDrawDebug=True,bDraw2D=False,bDrawLocalAxes=False,DebugScale=25,DebugSegments=20,DebugThickness=0.25)
    DriverAxis=(X=0,Y=1,Z=0)
    DriverItem=(Type=Bone,Name="head")
    FalloffRegionScaleFactors=(PositiveWidth=1,NegativeWidth=0.4,PositiveHeight=1,NegativeHeight=1)
    FalloffSize=0.2
    FlipHeightScaling=False
    FlipWidthScaling=False
    OptionalParentItem=(Type=Bone,Name="None")
    RotationOffset=(X=85,Y=0,Z=-20)
    ExecutePin=@RigUnit_BeginExecution.ExecutePin
   )
   > &RigUnit_ModifyTransforms(
    ItemToModify="((Item=(Type=Bone,Name=\"base_hair_12_m\"),Transform=(Rotation=(X=0,Y=-0.3277,Z=0,W=0.9448),Translation=(X=0.145,Y=0,Z=1.0428),Scale3D=(X=1,Y=1,Z=1))),(Item=(Type=Bone,Name=\"base_hair_11_m\"),Transform=(Rotation=(X=0,Y=-0.4435,Z=0,W=0.8963),Translation=(X=-0. ...(truncated)"
    WeightMinimum=0
    Weight=&RigVMFunction_MathDoubleMul(A=@n1.OutputParam, B=RigVMVariableNode() {variable=notCrouch}.Value).Result
   )
   > n2 = &RigUnit_SphericalPoseReader(
    ActiveRegionScaleFactors=(PositiveWidth=1.5,NegativeWidth=0.2,PositiveHeight=1,NegativeHeight=1)
    ActiveRegionSize=0.1
    Debug=(bDrawDebug=True,bDraw2D=False,bDrawLocalAxes=False,DebugScale=25,DebugSegments=20,DebugThickness=0.25)
    DriverAxis=(X=0,Y=1,Z=0)
    DriverItem=(Type=Bone,Name="head")
    FalloffRegionScaleFactors=(PositiveWidth=1,NegativeWidth=0.4,PositiveHeight=1,NegativeHeight=1)
    FalloffSize=0.2
    FlipHeightScaling=False
    FlipWidthScaling=False
    OptionalParentItem=(Type=Bone,Name="None")
    RotationOffset=(X=85,Y=0,Z=-32)
   )
   Multiply = &RigVMFunction_MathDoubleMul(A=@n2.OutputParam, B=RigVMVariableNode() {variable=crouchStatus}.Value)
   > &RigUnit_ModifyTransforms(
    ItemToModify="((Item=(Type=Bone,Name=\"base_hair_12_m\"),Transform=(Rotation=(X=0,Y=-0.3266,Z=0,W=0.9452),Translation=(X=-2.3271,Y=0.0001,Z=2.0701),Scale3D=(X=1,Y=1,Z=1))),(Item=(Type=Bone,Name=\"base_hair_11_m\"),Transform=(Rotation=(X=-0.0579,Y=-0.5419,Z=0.089,W=0.8337),T ...(truncated)"
    WeightMinimum=0
    Weight=@Multiply.Result
   )
  ]}
]
hierarchy [
 root {loc=<0 0 0>, rot=<0 0 0 1>}
 pelvis {parent=root, loc=<0 -0.5301 98.0571>, rot=<0.0127 0.707 -0.0127 -0.707>}
 spine_01 {parent=pelvis, loc=<6.1208 -0.9328 0>, rot=<0 0 0.034 -0.9994>}
 spine_02 {parent=spine_01, loc=<6.5996 -0.2095 0>, rot=<0 0 0.0238 -0.9997>}
 spine_03 {parent=spine_02, loc=<6.4914 -0.1237 0>, rot=<0 0 0.0149 -0.9999>}
 spine_04 {parent=spine_03, loc=<6.9598 0.0794 0>, rot=<0 0 -0.0132 -0.9999>}
 spine_05 {parent=spine_04, loc=<6.9987 0 0>, rot=<0 0 -0.0015 -1>}
 neck_01 {parent=spine_05, loc=<11.3138 0.1529 0>, rot=<0 0 -0.2003 -0.9797>}
 neck_02 {parent=neck_01, loc=<4.5627 0 0>, rot=<0 0 0.0504 -0.9987>}
 head {parent=neck_02, loc=<5.3914 0 0>, rot=<0 0 0.115 -0.9934>}
 base_hair_01_m {parent=head, loc=<8.9233 10.2143 2.6911>, rot=<0.0648 0.7041 0.7041 0.0648>}
 dyn_hair_01_m {parent=base_hair_01_m, loc=<0 0 0>, rot=<0 0 0 1>}
 base_hair_02_m {parent=head, loc=<8.1402 9.882 -0.7117>, rot=<0.1462 0.6915 0.7003 0.1001>}
 dyn_hair_02_m {parent=base_hair_02_m, loc=<0 0 0>, rot=<0 0 0 1>}
 base_hair_03_m {parent=head, loc=<14.9683 7.3208 2.3258>, rot=<0.542 0.3636 0.7012 0.2869>}
 dyn_hair_03_m {parent=base_hair_03_m, loc=<0 0 0>, rot=<0 0 0 1>}
 base_hair_04_m {parent=head, loc=<11.4286 3.3653 4.4074>, rot=<0.2991 0.1904 0.9333 0.057>}
 dyn_hair_04_m {parent=base_hair_04_m, loc=<0 0 0>, rot=<0 0 0 1>}
 base_hair_05_m {parent=head, loc=<16.5644 1.3229 1.1426>, rot=<0.6858 0.2815 0.472 0.4772>}
 dyn_hair_05_m {parent=base_hair_05_m, loc=<0 0 0>, rot=<0 0 0 1>}
 base_hair_06_m {parent=head, loc=<16.155 -3.7077 0.3655>, rot=<-0.0132 0.0496 -0.223 -0.9735>}
 dyn_hair_06_m {parent=base_hair_06_m, loc=<0 0 0>, rot=<0 0 0 1>}
 base_hair_07_m {parent=head, loc=<13.9221 -6.4607 -0.4581>, rot=<-0.8007 0.1141 0.1557 -0.5672>}
 dyn_hair_07_m {parent=base_hair_07_m, loc=<0 0 0>, rot=<0 0 0 1>}
 base_hair_08_m {parent=head, loc=<11.4819 -7.9599 -0.8493>, rot=<0.577 -0.2375 0.2259 -0.7481>}
 dyn_hair_08_m {parent=base_hair_08_m, loc=<0 0 0>, rot=<0 0 0 1>}
 base_hair_09_m {parent=head, loc=<10.512 -8.1698 0.9327>, rot=<0.3904 -0.6199 0.4444 -0.5156>}
 dyn_hair_09_m {parent=base_hair_09_m, loc=<0 0 0>, rot=<0 0 0 1>}
 base_hair_10_m {parent=head, loc=<6.3935 -9.5179 0.1908>, rot=<0.2604 -0.6386 0.6816 -0.2444>}
 dyn_hair_10_m {parent=base_hair_10_m, loc=<0 0 0>, rot=<0 0 0 1>}
 base_hair_11_m {parent=head, loc=<1.3289 -8.7118 -0.114>, rot=<0.1443 -0.6922 0.6922 -0.1443>}
 dyn_hair_11_m {parent=base_hair_11_m, loc=<0 0 0>, rot=<0 0 0 1>}
 base_hair_12_m {parent=head, loc=<-2.1225 -6.8352 0.1751>, rot=<0.2315 -0.6681 0.6618 -0.2489>}
 dyn_hair_12_m {parent=base_hair_12_m, loc=<0 0 0.8>, rot=<0 0.1305 0 -0.9914>}
]
variables [
 crouchStatus {type=double, default=0}
 notCrouch {type=double, default=0}
]
