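# Post-process style AnimBlueprint for the StrategicSpur Blast female face accessory part, adding secondary hair motion to the character.
# Copies the attached parent's pose, runs the part's Control Rig with crouch inputs, then caches it.
# Twelve chained root-relative AnimDynamics nodes simulate dyn_hair_01–12_m with angular springs and limits.
# The EventGraph updates crouch, gravity-override and zipline values, but the AnimGraph shows no reads of the latter two.
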AnimBlueprint /BRCosmetics/Characters/Player/Female/Medium/Bodies/F_MED_StrategicSpur_Blast/Meshes/Parts/F_MED_StrategicSpur_Blast_FaceAcc_AnimBP
target_skeleton=&/F_MED_StrategicSpur_Blast_FaceAcc_Skeleton
graphs [
 AnimGraph {nodes [
   n1 = Anim Dynamics - Bone: dyn_hair_01_m!(
    ComponentPose=Local To Component(LocalPose=Use cached pose 'CRCache_Input'())
    Alpha=1
    GravityOverride:Vector="0,0,-400"
    ExternalForce="0,0,0"
   ) -> (Pose) {
    angularDampingOverride=0.999
    angularSpringConstant=2000
    bAngularSpring=true
    bOverrideAngularDamping=true
    bUseGravityOverride=true
    gravityScale=0.35
    simulationSpace=RootRelative
    alphaBoolBlend {blendOption=Linear}
    boundBone {boneName=dyn_hair_01_m}
    componentAppliedLinearAccClamp=<5000 5000 5000>
    componentLinearAccScale=<0.5 0.5 0.5>
    componentLinearVelScale=<0.5 0.5 0.5>
    gravityOverride {z=-980}
    physicsBodyDefinitions [{
      boundBone {boneName=dyn_hair_01_m}
      boxExtents=<20 30 30>
      constraintSetup {angularLimitsMax=<15 10 40>, angularLimitsMin=<-15 -30 -40>, angularTarget=<1 0 0>}
      localJointOffset=<4 0 0>
     }]
    relativeSpaceBone {boneName=root}
   }
   n2 = Anim Dynamics - Bone: dyn_hair_02_m!(ComponentPose=@n1, Alpha=1, GravityOverride:Vector="0,0,-400", ExternalForce="0,0,0") -> (Pose) {
    angularDampingOverride=0.999
    angularSpringConstant=2000
    bAngularSpring=true
    bOverrideAngularDamping=true
    bUseGravityOverride=true
    gravityScale=0.35
    simulationSpace=RootRelative
    alphaBoolBlend {blendOption=Linear}
    boundBone {boneName=dyn_hair_02_m}
    componentAppliedLinearAccClamp=<5000 5000 5000>
    componentLinearAccScale=<0.5 0.5 0.5>
    componentLinearVelScale=<0.5 0.5 0.5>
    gravityOverride {z=-980}
    physicsBodyDefinitions [{
      boundBone {boneName=dyn_hair_02_m}
      boxExtents=<20 30 30>
      constraintSetup {angularLimitsMax=<15 15 40>, angularLimitsMin=<-15 -15 -40>, angularTarget=<1 0 0>}
      localJointOffset=<4 0 0>
     }]
    relativeSpaceBone {boneName=root}
   }
   n3 = Anim Dynamics - Bone: dyn_hair_03_m!(ComponentPose=@n2, Alpha=1, GravityOverride:Vector="0,0,-400", ExternalForce="0,0,0") -> (Pose) {
    angularDampingOverride=0.999
    angularSpringConstant=2000
    bAngularSpring=true
    bOverrideAngularDamping=true
    bUseGravityOverride=true
    gravityScale=0.35
    simulationSpace=RootRelative
    alphaBoolBlend {blendOption=Linear}
    boundBone {boneName=dyn_hair_03_m}
    componentAppliedLinearAccClamp=<5000 5000 5000>
    componentLinearAccScale=<0.5 0.5 0.5>
    componentLinearVelScale=<0.5 0.5 0.5>
    gravityOverride {z=-980}
    physicsBodyDefinitions [{
      boundBone {boneName=dyn_hair_03_m}
      boxExtents=<20 30 30>
      constraintSetup {angularLimitsMax=<0 20 40>, angularLimitsMin=<0 -20 -40>, angularTarget=<1 0 0>}
      localJointOffset=<4 0 0>
     }]
    relativeSpaceBone {boneName=root}
   }
   n4 = Anim Dynamics - Bone: dyn_hair_04_m!(ComponentPose=@n3, Alpha=1, GravityOverride:Vector="0,0,-400", ExternalForce="0,0,0") -> (Pose) {
    angularDampingOverride=0.999
    angularSpringConstant=10000
    bAngularSpring=true
    bOverrideAngularDamping=true
    bUseGravityOverride=true
    gravityScale=0.35
    simulationSpace=RootRelative
    alphaBoolBlend {blendOption=Linear}
    boundBone {boneName=dyn_hair_04_m}
    componentAppliedLinearAccClamp=<5000 5000 5000>
    componentLinearAccScale=<0.5 0.5 0.5>
    componentLinearVelScale=<0.5 0.5 0.5>
    gravityOverride {z=-980}
    physicsBodyDefinitions [{
      boundBone {boneName=dyn_hair_04_m}
      boxExtents=<20 30 30>
      constraintSetup {angularLimitsMax=<0 1 20>, angularLimitsMin=<0 -15 -20>, angularTarget=<1 0 0>}
      localJointOffset=<4 0 0>
     }]
    relativeSpaceBone {boneName=root}
   }
   n5 = Anim Dynamics - Bone: dyn_hair_05_m!(ComponentPose=@n4, Alpha=1, GravityOverride:Vector="0,0,-400", ExternalForce="0,0,0") -> (Pose) {
    angularDampingOverride=0.999
    angularSpringConstant=2000
    bAngularSpring=true
    bOverrideAngularDamping=true
    bUseGravityOverride=true
    gravityScale=0.35
    simulationSpace=RootRelative
    alphaBoolBlend {blendOption=Linear}
    boundBone {boneName=dyn_hair_05_m}
    componentAppliedLinearAccClamp=<5000 5000 5000>
    componentLinearAccScale=<0.5 0.5 0.5>
    componentLinearVelScale=<0.5 0.5 0.5>
    gravityOverride {z=-980}
    physicsBodyDefinitions [{
      boundBone {boneName=dyn_hair_05_m}
      boxExtents=<20 30 30>
      constraintSetup {angularLimitsMax=<15 15 40>, angularLimitsMin=<-15 -15 -40>, angularTarget=<1 0 0>}
      localJointOffset=<4 0 0>
     }]
    relativeSpaceBone {boneName=root}
   }
   n6 = Anim Dynamics - Bone: dyn_hair_06_m!(ComponentPose=@n5, Alpha=1, GravityOverride:Vector="0,0,-400", ExternalForce="0,0,0") -> (Pose) {
    angularDampingOverride=0.999
    angularSpringConstant=10000
    bAngularSpring=true
    bOverrideAngularDamping=true
    bUseGravityOverride=true
    gravityScale=0.35
    simulationSpace=RootRelative
    alphaBoolBlend {blendOption=Linear}
    boundBone {boneName=dyn_hair_06_m}
    componentAppliedLinearAccClamp=<5000 5000 5000>
    componentLinearAccScale=<0.5 0.5 0.5>
    componentLinearVelScale=<0.5 0.5 0.5>
    gravityOverride {z=-980}
    physicsBodyDefinitions [{
      boundBone {boneName=dyn_hair_06_m}
      boxExtents=<20 30 30>
      constraintSetup {angularLimitsMax=<0 40 20>, angularLimitsMin=<0 -40 -20>, angularTarget=<1 0 0>}
      localJointOffset=<4 0 0>
     }]
    relativeSpaceBone {boneName=root}
   }
   n7 = Anim Dynamics - Bone: dyn_hair_07_m!(ComponentPose=@n6, Alpha=1, GravityOverride:Vector="0,0,-400", ExternalForce="0,0,0") -> (Pose) {
    angularDampingOverride=0.999
    angularSpringConstant=2000
    bAngularSpring=true
    bOverrideAngularDamping=true
    bUseGravityOverride=true
    gravityScale=0.35
    simulationSpace=RootRelative
    alphaBoolBlend {blendOption=Linear}
    boundBone {boneName=dyn_hair_07_m}
    componentAppliedLinearAccClamp=<5000 5000 5000>
    componentLinearAccScale=<0.5 0.5 0.5>
    componentLinearVelScale=<0.5 0.5 0.5>
    physicsBodyDefinitions [{
      boundBone {boneName=dyn_hair_07_m}
      boxExtents=<20 30 30>
      constraintSetup {angularLimitsMax=<15 15 40>, angularLimitsMin=<-15 -15 -40>, angularTarget=<1 0 0>}
      localJointOffset=<4 0 0>
     }]
    relativeSpaceBone {boneName=root}
   }
   n8 = Anim Dynamics - Bone: dyn_hair_08_m!(ComponentPose=@n7, Alpha=1, GravityOverride:Vector="0,0,-400", ExternalForce="0,0,0") -> (Pose) {
    angularDampingOverride=0.999
    angularSpringConstant=2000
    bAngularSpring=true
    bOverrideAngularDamping=true
    bUseGravityOverride=true
    gravityScale=0.35
    simulationSpace=RootRelative
    alphaBoolBlend {blendOption=Linear}
    boundBone {boneName=dyn_hair_08_m}
    componentAppliedLinearAccClamp=<5000 5000 5000>
    componentLinearAccScale=<0.5 0.5 0.5>
    componentLinearVelScale=<0.5 0.5 0.5>
    physicsBodyDefinitions [{
      boundBone {boneName=dyn_hair_08_m}
      boxExtents=<20 30 30>
      constraintSetup {angularLimitsMax=<15 15 40>, angularLimitsMin=<-15 -15 -40>, angularTarget=<1 0 0>}
      localJointOffset=<4 0 0>
     }]
    relativeSpaceBone {boneName=root}
   }
   n9 = Anim Dynamics - Bone: dyn_hair_09_m!(ComponentPose=@n8, Alpha=1, GravityOverride:Vector="0,0,-400", ExternalForce="0,0,0") -> (Pose) {
    angularDampingOverride=0.999
    angularSpringConstant=2000
    bAngularSpring=true
    bOverrideAngularDamping=true
    bUseGravityOverride=true
    gravityScale=0.35
    simulationSpace=RootRelative
    alphaBoolBlend {blendOption=Linear}
    boundBone {boneName=dyn_hair_09_m}
    componentAppliedLinearAccClamp=<5000 5000 5000>
    componentLinearAccScale=<0.5 0.5 0.5>
    componentLinearVelScale=<0.5 0.5 0.5>
    physicsBodyDefinitions [{
      boundBone {boneName=dyn_hair_09_m}
      boxExtents=<20 30 30>
      constraintSetup {angularLimitsMax=<0 20 40>, angularLimitsMin=<0 -20 -40>, angularTarget=<1 0 0>}
      localJointOffset=<4 0 0>
     }]
    relativeSpaceBone {boneName=root}
   }
   n10 = Anim Dynamics - Bone: dyn_hair_10_m!(ComponentPose=@n9, Alpha=1, GravityOverride:Vector="0,0,-400", ExternalForce="0,0,0") -> (Pose) {
    angularDampingOverride=0.999
    angularSpringConstant=2000
    bAngularSpring=true
    bOverrideAngularDamping=true
    bUseGravityOverride=true
    gravityScale=0.35
    simulationSpace=RootRelative
    alphaBoolBlend {blendOption=Linear}
    boundBone {boneName=dyn_hair_10_m}
    componentAppliedLinearAccClamp=<5000 5000 5000>
    componentLinearAccScale=<0.5 0.5 0.5>
    componentLinearVelScale=<0.5 0.5 0.5>
    physicsBodyDefinitions [{
      boundBone {boneName=dyn_hair_10_m}
      boxExtents=<20 30 30>
      constraintSetup {angularLimitsMax=<15 35 40>, angularLimitsMin=<-15 -35 -40>, angularTarget=<1 0 0>}
      localJointOffset=<4 0 0>
     }]
    relativeSpaceBone {boneName=root}
   }
   n11 = Anim Dynamics - Bone: dyn_hair_11_m!(ComponentPose=@n10, Alpha=1, GravityOverride:Vector="0,0,-400", ExternalForce="0,0,0") -> (Pose) {
    angularDampingOverride=0.999
    angularSpringConstant=2000
    bAngularSpring=true
    bOverrideAngularDamping=true
    bUseGravityOverride=true
    gravityScale=0.35
    simulationSpace=RootRelative
    alphaBoolBlend {blendOption=Linear}
    boundBone {boneName=dyn_hair_11_m}
    componentAppliedLinearAccClamp=<5000 5000 5000>
    componentLinearAccScale=<0.5 0.5 0.5>
    componentLinearVelScale=<0.5 0.5 0.5>
    physicsBodyDefinitions [{
      boundBone {boneName=dyn_hair_11_m}
      boxExtents=<20 30 30>
      constraintSetup {angularLimitsMax=<15 15 40>, angularLimitsMin=<-15 -35 -40>, angularTarget=<1 0 0>}
      localJointOffset=<4 0 0>
     }]
    relativeSpaceBone {boneName=root}
   }
   n12 = Anim Dynamics - Bone: dyn_hair_12_m!(ComponentPose=@n11, Alpha=1, GravityOverride:Vector="0,0,-400", ExternalForce="0,0,0") -> (Pose) {
    angularDampingOverride=0.999
    angularSpringConstant=2000
    bAngularSpring=true
    bOverrideAngularDamping=true
    bUseGravityOverride=true
    gravityScale=0.35
    simulationSpace=RootRelative
    alphaBoolBlend {blendOption=Linear}
    boundBone {boneName=dyn_hair_12_m}
    componentAppliedLinearAccClamp=<5000 5000 5000>
    componentLinearAccScale=<0.5 0.5 0.5>
    componentLinearVelScale=<0.5 0.5 0.5>
    physicsBodyDefinitions [{
      boundBone {boneName=dyn_hair_12_m}
      boxExtents=<20 30 30>
      constraintSetup {angularLimitsMax=<0 20 40>, angularLimitsMin=<0 -35 -40>, angularTarget=<1 0 0>}
      localJointOffset=<4 0 0>
     }]
    relativeSpaceBone {boneName=root}
   }
   Output Pose(Result=Component To Local(ComponentPose=@n12)) {layerGroup=None}
   n13 = Control Rig!(
    Source=Copy Pose From Mesh() {bUseAttachedParent=true}
    Alpha
    notCrouch:double=Get NotCrouching() -> (NotCrouching:double)
    crouchStatus:double=Get CrouchAlpha() -> (CrouchAlpha:double)
   ) -> (Pose) {
    alphaBoolBlend {blendOption=Linear}
    controlRigAssetReference {blueprintRigClass=&/F_MED_StrategicSpur_Blast_FaceAcc_CtrlRig.F_MED_StrategicSpur_Blast_FaceAcc_CtrlRig_C}
   }
   Save cached pose 'CRCache_Input'(Pose=@n13) {cacheName=CRCache_Input}
  ]}
 EventGraph {nodes [
   Event BlueprintUpdateAnimation()
   > n1 = Sequence!(execute) -> (then_0, then_1, then_2:Exec, then_3:Exec)
   Set CrouchAlpha(execute=@n1.then_1, CrouchAlpha:double="To Float (Boolean)"(InBool=Get bIsCrouching())) -> (Output_Get:double)
   n2 = Gravity Override State Function DEPRECATED(
    execute=@n1.then_0
    CharacterPartAnimInstance=Self-Reference()
    GameTimeInSeconds=GetGameTimeInSeconds()
    GravityOverrideParams=Get Gravity Override Params()
   )
   > Set GravityOverride(GravityOverride=@n2)
   n3 = Branch(execute=@n1.then_2, Condition=Get bIsCrouching())
   > Set NotCrouching(NotCrouching:double=0) -> (Output_Get:double)
   Set NotCrouching(execute=@n3.else, NotCrouching=1)
   n4 = Branch(execute=@n1.then_3, Condition=Get OnZipline() -> (OnZipline:bool))
   > Set ZiplineAlpha(ZiplineAlpha:double=0) -> (Output_Get:double)
   Set ZiplineAlpha(execute=@n4.else, ZiplineAlpha=1)
   n5 = Event BlueprintProcessAnimInputEvent()
   > n6 = Branch(Condition=IsNamed(self=@n5.AnimInputEvent, InEventName=aie.Player.Zipline.Entered))
   > Set OnZipline(OnZipline:bool=1) -> (Output_Get:bool)
   Branch(execute=@n6.else, Condition=IsNamed(self=@n5.AnimInputEvent, InEventName=aie.Player.Zipline.Exited))
   > Set OnZipline(OnZipline=false)
  ]}
]
variables [
 CrouchAlpha {type=number}
 Gravity Override Params {type=GravityOverrideParamsStruct}
 GravityOverride {type=Vector}
 NewVar {type=integer}
 NotCrouching {type=number}
 OnZipline {type=boolean}
 ZiplineAlpha {type=number}
]
